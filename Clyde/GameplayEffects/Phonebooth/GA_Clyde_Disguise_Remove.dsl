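; Gameplay ability that removes the Clyde phonebooth disguise in response to gameplay events.
; On activation it removes owner effects granting the DishFries active tag, then calls the parent.
; ShouldAbilityRespondToEvent switches on event tag to filter triggers, with separate cases for weakspot and melee.
; Damage-dealt events are checked for nonzero magnitude and a target that is either a player pawn or tagged as a sentry.
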(event Ability|EventActivateAbilityFromEvent
  (params (EventData :type Gameplay%20Event%20Data%20Structure%20%28by%20ref%29))
  (Ability|RemoveGameplayEffectFromOwnerWithGrantedTags "(GameplayTags=((TagName=\"Granted.Athena.DishFries.Active\")))")
  (|CallParentFunction EventData))

(fn ShouldAbilityRespondToEvent
  (params (ActorInfo :type Gameplay%20Ability%20Actor%20Info%20Structure) (Payload :type Gameplay%20Event%20Data%20Structure))
  (bind (_eventtag _instigator _target _optionalobject _optionalobject2 _contexthandle _instigatortags _targettags _eventmagnitude) (Utilities|Struct|BreakGameplayEventData Payload))
  ;[1 1 1] If the event was the damage event check if it is valid target
  (Utilities|FlowControl|Sequence
    (:then_0
      (switch Utilities|FlowControl|Switch|SwitchonGameplayTag _eventtag
        (:Default
          (LocalVariables|SetShould%20Respond true))
        (:Event.Damage.Dealt.Weakspot
          _)
        (:Event.Damage.Dealt
          (if (not (== _eventmagnitude 0.0))
            (Utilities|Casting|CastToFortPlayerPawn _target
              (:then
                (LocalVariables|SetShould%20Respond true))
              (:CastFailed
                (bind _returnvalue (GameplayTags|HasMatchingGameplayTag(Message) _target "(TagName=\"Asset.Athena.EnvItem.Sentry\")"))
                (return _returnvalue)))))
        (:Abilities.Weapons.Melee
          _))
      (LocalVariables|SetShould%20Respond false))
    (:then_1
      (return (LocalVariables|GetShould%20Respond)))))
